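# Translucent UI master material that draws a pip with a soft glow for the Homing Rocket widgets.
# Two rounded-square gradients form a hardened pip and an offset, tapered glow, with the pip mask blending their colors and alphas.
# CompletedState lerps between the normal and completed color pairs.
# The HasStripes static switch modulates opacity with animated stripes, built from square-to-screen-space UVs that are rotated, panned and fed through a hardened sine.
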Material /ResolveGameplay/Items/Guns/HomingRocket/Widgets/UI_Materials/M_UI_HomingRocket_Pip_Master
graph {
 properties {blendMode=BLEND_Translucent, materialDomain=MD_UI, pixelDepthOffsetMode=PDOM_Legacy, usedShadingModels=Default Lit}
 nodes [
  n1 = MaterialFunctionCall!(
   UV
   Size=ScalarParameter() {defaultValue=0.5, group=Pip, parameterName=Pip_Size}
   Curvature=ScalarParameter() {defaultValue=0.25, group=Pip, parameterName=Pip_CornerRadius}
  ) {materialFunction=/Game/Effects/Fort_Effects/MaterialFunctions/MF_RoundedSquare}
  n2 = Saturate(None=Multiply(A=Saturate(None=@n1.Linear Gradient), B=ScalarParameter() {defaultValue=4, group=Pip, parameterName=Pip_EdgeHardness}))
  LinearInterpolate(
   A=LinearInterpolate(
    A=VectorParameter() {parameterName=Color_Glow, defaultValue=<rgba 0 1 0 1>}.RGB
    B=VectorParameter() {parameterName=Color_Pip, defaultValue=<rgba 1 0 0 1>}.RGB
    Alpha=@n2
   )
   B=LinearInterpolate(
    A=VectorParameter() {parameterName=CompletedColor_Glow, defaultValue=<rgba 0 0 0 1>}.RGB
    B=VectorParameter() {parameterName=CompletedColor_Pip, defaultValue=<rgba 0 0 0 1>}.RGB
    Alpha=@n2
   )
   Alpha=ScalarParameter() {parameterName=CompletedState}
  )
  n3 = ScalarParameter() {group=Glow, parameterName=Glow_OffsetX}
  n4 = ScalarParameter() {group=Glow, parameterName=Glow_OffsetY}
  n5 = MaterialFunctionCall(
   UV=Add(A=TextureCoordinate(), B=AppendVector(A=@n3, B=@n4))
   Size=ScalarParameter() {defaultValue=1, group=Glow, parameterName=Glow_Size}
   Curvature=ScalarParameter() {defaultValue=0.75, group=Glow, parameterName=Glow_CornerRadius}
  ) {materialFunction=&MF_RoundedSquare}
  n6 = Saturate(None=Power(Base=Saturate(None=@n5.Linear Gradient), Exp=ScalarParameter() {defaultValue=1.22, group=Glow, parameterName=Glow_Taper}))
  n7 = ScalarParameter() {defaultValue=0.5, group=Glow, parameterName=Alpha_Glow}
  n8 = Saturate(None=LinearInterpolate(A=Multiply(A=@n6, B=@n7), B=ScalarParameter() {defaultValue=1, group=Pip, parameterName=Alpha_Pip}, Alpha=@n2))
  n9 = ScalarParameter() {defaultValue=0.5, parameterName=StripesGapAlpha}
  n10 = Multiply(
   A=MaterialFunctionCall() {materialFunction=/FortUILibrary/Foundation/Art_ToolKit/Tools/MF_UI_SquareToScreenSpace}.X_Priority
   B=ScalarParameter() {defaultValue=16, parameterName=StripesTile}
  )
  n11 = ScalarParameter() {defaultValue=-0.125, parameterName=StripesRotation}
  n12 = ScalarParameter() {defaultValue=-0.05, parameterName=PanSpeedX}
  n13 = Constant()
  n14 = Sine(None=Panner(
    Coordinate=MaterialFunctionCall(UVs=@n10, "Rotation Angle (0-1)"=@n11) {materialFunction=&CustomRotator}.Rotated Values
    Time=Time()
    Speed=AppendVector(A=@n12, B=@n13)
   ))
  n15 = Saturate(None=Multiply(A=ComponentMask(None=@n14) {r=true}, B=ScalarParameter() {defaultValue=4, parameterName=StripesHardness}))
  n16 = Saturate(None=DotProduct(A=@n8, B=LinearInterpolate(A=@n9, Alpha=@n15)))
  Saturate(None=Multiply(A=StaticSwitchParameter(True=@n16, False=@n8) {parameterName=HasStripes}, B=ScalarParameter() {defaultValue=1, parameterName=Final_Alpha}))
 ]
}
