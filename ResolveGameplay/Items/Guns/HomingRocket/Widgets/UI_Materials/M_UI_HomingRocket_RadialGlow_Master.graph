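# Translucent UI master material for a radial glow in the homing rocket's widgets.
# An exponential radial gradient with adjustable hardness feeds two separate power curves, one to taper color and one to taper alpha.
# The color curve drives a three-color blend from falloff through middle to hotspot.
# The alpha curve is scaled by a final alpha parameter, so instances can tune color spread and transparency independently.
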Material /ResolveGameplay/Items/Guns/HomingRocket/Widgets/UI_Materials/M_UI_HomingRocket_RadialGlow_Master
graph {
 properties {blendMode=BLEND_Translucent, materialDomain=MD_UI, pixelDepthOffsetMode=PDOM_Legacy, usedShadingModels=Default Lit}
 nodes [
  n1 = MaterialFunctionCall(UVs=TextureCoordinate(), Density=ScalarParameter() {defaultValue=1.666, parameterName=Hardness}) {materialFunction=&RadialGradientExponential}
  n2 = ScalarParameter() {defaultValue=1.22, parameterName=TaperColorGradient}
  MaterialFunctionCall(
   A=VectorParameter() {parameterName=ColorFalloff, defaultValue=<rgba 1 0 0 1>}.RGB
   B=VectorParameter() {parameterName=ColorMiddle, defaultValue=<rgba 0 1 0 1>}.RGB
   C=VectorParameter() {parameterName=ColorHotspot, defaultValue=<rgba 0 0 1 1>}.RGB
   Alpha=Saturate(None=Power(Base=@n1.RadialGradientExponential, Exp=@n2))
  ) {materialFunction=&3ColorBlend}
  n3 = ScalarParameter() {defaultValue=1, parameterName=TaperAlphaGradient}
  Multiply(A=Saturate(None=Power(Base=@n1.RadialGradientExponential, Exp=@n3)), B=ScalarParameter() {defaultValue=1, parameterName=Final_Alpha})
 ]
}
