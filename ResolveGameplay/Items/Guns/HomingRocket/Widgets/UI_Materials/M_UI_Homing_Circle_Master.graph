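# Translucent UI master material for the homing rocket gun's widgets, drawing a procedural circle ring in a parameterized color.
# Two RadialGradientExponential calls form the ring mask: the outer gradient times the inverted inner one (via DotProduct), saturated, with radius and density exposed for each.
# The mask blends ColorA to ColorB and, scaled by Final_Alpha and saturated, also yields an alpha value.
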Material /ResolveGameplay/Items/Guns/HomingRocket/Widgets/UI_Materials/M_UI_Homing_Circle_Master
graph {
 properties {blendMode=BLEND_Translucent, materialDomain=MD_UI, pixelDepthOffsetMode=PDOM_Legacy, usedShadingModels=Default Lit}
 nodes [
  n1 = ScalarParameter() {defaultValue=0.5, parameterName=OutsideRadius}
  n2 = ScalarParameter() {defaultValue=8, parameterName=OutsideDensity}
  n3 = MaterialFunctionCall(
   Radius=ScalarParameter() {defaultValue=0.4, parameterName=InsideRadius}
   Density=ScalarParameter() {defaultValue=8, parameterName=InsideDensity}
  ) {materialFunction=&RadialGradientExponential}
  n4 = Saturate(None=DotProduct(
    A=MaterialFunctionCall(Radius=@n1, Density=@n2) {materialFunction=&RadialGradientExponential}.RadialGradientExponential
    B=OneMinus(None=@n3.RadialGradientExponential)
   ))
  LinearInterpolate(
   A=VectorParameter() {parameterName=ColorA, defaultValue=<rgba 1 0 0 1>}.RGB
   B=VectorParameter() {parameterName=ColorB, defaultValue=<rgba 0 1 0 1>}.RGB
   Alpha=@n4
  )
  Saturate(None=Multiply(A=@n4, B=ScalarParameter() {defaultValue=1, parameterName=Final_Alpha}))
 ]
}
