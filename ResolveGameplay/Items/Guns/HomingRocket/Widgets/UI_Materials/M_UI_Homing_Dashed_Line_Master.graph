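# Translucent UI master material for a dashed line, filed under the Homing Rocket gun's widget materials.
# A sine of scaled UVs, masked to the V axis and clamped, forms repeating bands sized by Lines.
# SmoothStep, EdgeHardness and FinalAlpha shape those bands into an alpha-style mask.
# A separate lerp blends the Color_1 and Color_2 parameters with no alpha input wired, so the mix does not follow the dashes.
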Material /ResolveGameplay/Items/Guns/HomingRocket/Widgets/UI_Materials/M_UI_Homing_Dashed_Line_Master
graph {
 properties {blendMode=BLEND_Translucent, materialDomain=MD_UI, pixelDepthOffsetMode=PDOM_Legacy, usedShadingModels=Default Lit}
 nodes [
  LinearInterpolate(
   A=VectorParameter() {parameterName=Color_2, defaultValue=<rgba 1 0 0 1>}.RGB
   B=VectorParameter() {parameterName=Color_1, defaultValue=<rgba 1 0 0 1>}.RGB
  )
  n1 = Sine(None=Multiply(A=TextureCoordinate(), B=ScalarParameter() {defaultValue=16, parameterName=Lines})) {period=2}
  n2 = MaterialFunctionCall(
   Alpha=Saturate(None=ComponentMask(None=@n1) {g=true})
   Min=ScalarParameter() {parameterName=Min}
   Max=ScalarParameter() {defaultValue=1, parameterName=Max}
  ) {materialFunction=&SmoothStep}
  n3 = ScalarParameter() {defaultValue=1, parameterName=EdgeHardness}
  Multiply(A=Saturate(None=Multiply(A=@n2.Result, B=@n3)), B=ScalarParameter() {defaultValue=1, parameterName=FinalAlpha})
 ]
}
