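# Master UI-domain translucent material for font in the homing rocket gun's widgets, blending color and alpha between start and stop states.
# A scalar Progress parameter drives two linear interpolations: ColorStart to ColorStop for RGB, and AlphaStart to AlphaStop for alpha.
# All endpoints are exposed as parameters, with both colors defaulting to red and both alphas to fully opaque.
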Material /ResolveGameplay/Items/Guns/HomingRocket/Widgets/UI_Materials/M_UI_Homing_Font_Master
graph {
 properties {blendMode=BLEND_Translucent, materialDomain=MD_UI, pixelDepthOffsetMode=PDOM_Legacy, usedShadingModels=Default Lit}
 nodes [
  n1 = ScalarParameter() {parameterName=Progress}
  LinearInterpolate(
   A=VectorParameter() {parameterName=ColorStart, defaultValue=<rgba 1 0 0 1>}.RGB
   B=VectorParameter() {parameterName=ColorStop, defaultValue=<rgba 1 0 0 1>}.RGB
   Alpha=@n1
  )
  LinearInterpolate(A=ScalarParameter() {defaultValue=1, parameterName=AlphaStart}, B=ScalarParameter() {defaultValue=1, parameterName=AlphaStop}, Alpha=@n1)
 ]
}
