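# Translucent UI master material for the homing rocket, drawing an animated tiled pattern between two parameterised colours.
# Panned UI UVs are quantised into tile indices, whose vector length times time drives a cosine flicker.
# A looping horizontal sine sweep masks the flicker, producing the colour blend and alpha.
# The apparent aspect correction divides pixel height by itself, so it has no effect.
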Material /ResolveGameplay/Items/Guns/HomingRocket/Widgets/UI_Materials/M_UI_Homing_Pattern_Master
graph {
 properties {blendMode=BLEND_Translucent, materialDomain=MD_UI, pixelDepthOffsetMode=PDOM_Legacy, usedShadingModels=Default Lit}
 nodes [
  n1 = MaterialFunctionCall() {materialFunction=&GetUserInterfaceUV}
  n2 = OneMinus(None=@n1.Normalized UV)
  n3 = MaterialFunctionCall(Float2=@n1.Pixel Size) {materialFunction=&BreakOutFloat2Components}
  n4 = Divide(A=@n3.G, B=@n3.G)
  n5 = AppendVector(A=Constant() {r=1}, B=@n4)
  n6 = Constant()
  n7 = Multiply(A=OneMinus(None=@n4)) {constB=0.5}
  n8 = Panner(
   Coordinate=Add(A=Multiply(A=@n2, B=@n5), B=AppendVector(A=@n6, B=@n7))
   Speed=AppendVector(
    A=ScalarParameter() {defaultValue=0.25, group=Tiling, parameterName=PanSpeed_SolidTiles_X}
    B=ScalarParameter() {group=Tiling, parameterName=PanSpeed_SolidTiles_Y}
   )
  ) {speedX=0.5}
  n9 = Ceil(None=Multiply(A=Frac(None=@n8), B=ScalarParameter() {defaultValue=2, group=Tiling, parameterName=Tile}))
  n10 = MaterialFunctionCall(Vector 2=@n9) {materialFunction=/Engine/Functions/Engine_MaterialFunctions02/Utility/VectorLength}
  n11 = Multiply(A=Time(), B=ScalarParameter() {defaultValue=0.125, group=Tiling, parameterName=Tile_AnimSpeed})
  n12 = LinearInterpolate(
   A=ScalarParameter() {defaultValue=0.125, group=Tiling, parameterName=MoreConsistentTiles}
   Alpha=Cosine(None=Multiply(A=@n10.V2 Length, B=@n11))
  )
  n13 = Time()
  n14 = ScalarParameter() {defaultValue=0.75, parameterName=PanSpeed}
  n15 = LinearInterpolate(
   A=ScalarParameter() {defaultValue=-1, parameterName=PanPosition_Start}
   B=ScalarParameter() {defaultValue=1, parameterName=PanPosition_Stop}
   Alpha=Fmod(A=Multiply(A=@n13, B=@n14), B=ScalarParameter() {defaultValue=1, parameterName=FmodStop})
  )
  n16 = Sine(None=Add(A=TextureCoordinate(), B=OneMinus(None=@n15))) {period=2}
  n17 = Saturate(None=DotProduct(A=Saturate(None=@n12), B=ComponentMask(None=@n16) {r=true}))
  LinearInterpolate(
   A=VectorParameter() {parameterName=ColorA, defaultValue=<rgba 0 1 0 1>}.RGB
   B=VectorParameter() {parameterName=ColorB, defaultValue=<rgba 1 0 0 1>}.RGB
   Alpha=@n17
  )
  Multiply(A=@n9, B=ScalarParameter() {defaultValue=2, parameterName=Random})
  Multiply(A=@n17, B=ScalarParameter() {defaultValue=1, parameterName=Final_Alpha})
 ]
}
