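# Translucent UI master material for the homing rocket widget, showing lock-on progress as a fill over a backing.
# A horizontal mask from UV channel 4, the Progress parameter and an edge-hardness multiplier blends backing and progress color and alpha.
# IsLockingOn and IsLockedOn scalars pick each layer's default, locking-on or locked-on color.
# The IsProgress static switch reduces it to backing only when off, and Final_Alpha scales overall opacity.
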Material /ResolveGameplay/Items/Guns/HomingRocket/Widgets/UI_Materials/M_UI_Homing_Progress_Master
graph {
 properties {blendMode=BLEND_Translucent, materialDomain=MD_UI, pixelDepthOffsetMode=PDOM_Legacy, usedShadingModels=Default Lit}
 nodes [
  n1 = ScalarParameter() {defaultValue=0.5, parameterName=Alpha_Backing}
  n2 = TextureCoordinate() {coordinateIndex=4}
  n3 = Add(A=Add(A=OneMinus(None=@n2)) {constB=-1}, B=ScalarParameter() {defaultValue=0.2, parameterName=Progress})
  n4 = Saturate(None=Multiply(A=ComponentMask(None=@n3) {r=true}, B=ScalarParameter() {defaultValue=128, parameterName=Progress_EdgeHardness}))
  n5 = LinearInterpolate(A=@n1, B=ScalarParameter() {defaultValue=1, parameterName=Alpha_Progress}, Alpha=@n4)
  Saturate(None=Multiply(A=StaticSwitchParameter(True=@n5, False=@n1) {parameterName=IsProgress}, B=ScalarParameter() {defaultValue=1, parameterName=Final_Alpha}))
  n6 = LinearInterpolate(
   A=VectorParameter() {parameterName=Color_Backing, defaultValue=<rgba 0 1 0 1>}.RGB
   B=LinearInterpolate(
    A=VectorParameter() {parameterName=Color_Backing_LockingOn, defaultValue=<rgba 0 1 0 1>}.RGB
    B=VectorParameter() {parameterName=Color_Backing_LockedOn, defaultValue=<rgba 0 1 0 1>}.RGB
    Alpha=ScalarParameter() {parameterName=IsLockedOn}
   )
   Alpha=ScalarParameter() {parameterName=IsLockingOn}
  )
  n7 = VectorParameter() {parameterName=Color_Progress, defaultValue=<rgba 1 0 0 1>}
  n8 = LinearInterpolate(
   A=VectorParameter() {parameterName=Color_Progress_LockingOn, defaultValue=<rgba 1 0 0 1>}.RGB
   B=VectorParameter() {parameterName=Color_Progress_LockedOn, defaultValue=<rgba 1 0 0 1>}.RGB
   Alpha=ScalarParameter() {parameterName=IsLockedOn}
  )
  n9 = ScalarParameter() {parameterName=IsLockingOn}
  StaticSwitchParameter(True=LinearInterpolate(A=@n6, B=LinearInterpolate(A=@n7.RGB, B=@n8, Alpha=@n9), Alpha=@n4), False=@n6) {parameterName=IsProgress}
 ]
}
