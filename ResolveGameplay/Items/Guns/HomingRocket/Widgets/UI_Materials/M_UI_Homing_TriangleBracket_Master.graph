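# Translucent UI master material for a triangle bracket shape in the homing rocket's widgets.
# Four parameterized, rotatable linear gradients form the mask: edges 1 and 2 multiply together, and the inverted product of edges 3 and 4 cuts into it, using DotProduct as the multiply.
# A static switch optionally fades the mask with a screen-aligned radial vignette.
# The mask, scaled by Final_Alpha, blends between ColorA and ColorB.
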Material /ResolveGameplay/Items/Guns/HomingRocket/Widgets/UI_Materials/M_UI_Homing_TriangleBracket_Master
graph {
 properties {blendMode=BLEND_Translucent, materialDomain=MD_UI, pixelDepthOffsetMode=PDOM_Legacy, usedShadingModels=Default Lit}
 nodes [
  n1 = ScalarParameter() {parameterName=Edge1_RotationCenter_X}
  n2 = ScalarParameter() {parameterName=Edge1_RotationCenter_Y}
  n3 = ScalarParameter() {parameterName=Edge2_RotationCenter_X}
  n4 = ScalarParameter() {parameterName=Edge2_RotationCenter_Y}
  n5 = DotProduct(
   A=MaterialFunctionCall(
    GradientStart=ScalarParameter() {defaultValue=0.125, parameterName=Edge1_Start}
    GradientEnd=ScalarParameter() {defaultValue=0.5, parameterName=Edge1_Stop}
    RotationAngle=ScalarParameter() {parameterName=Edge1_Rotation}
    RotationCenter=AppendVector(A=@n1, B=@n2)
   ) {materialFunction=&MF_UI_LinearGradient}.Result
   B=MaterialFunctionCall(
    GradientStart=ScalarParameter() {defaultValue=0.125, parameterName=Edge2_Start}
    GradientEnd=ScalarParameter() {defaultValue=0.5, parameterName=Edge2_Stop}
    RotationAngle=ScalarParameter() {defaultValue=-0.25, parameterName=Edge2_Rotation}
    RotationCenter=AppendVector(A=@n3, B=@n4)
   ) {materialFunction=&MF_UI_LinearGradient}.Result
  )
  n6 = MaterialFunctionCall(
   GradientStart=ScalarParameter() {defaultValue=0.5, parameterName=Edge3_Start}
   GradientEnd=ScalarParameter() {defaultValue=0.75, parameterName=Edge3_Stop}
   RotationAngle=ScalarParameter() {defaultValue=0.425, parameterName=Edge3_Rotation}
   RotationCenter=AppendVector(
    A=ScalarParameter() {defaultValue=0.5, parameterName=Edge3_RotationCenter_X}
    B=ScalarParameter() {defaultValue=0.5, parameterName=Edge3_RotationCenter_Y}
   )
  ) {materialFunction=&MF_UI_LinearGradient}
  n7 = MaterialFunctionCall(
   GradientStart=ScalarParameter() {defaultValue=0.5, parameterName=Edge4_Start}
   GradientEnd=ScalarParameter() {defaultValue=0.75, parameterName=Edge4_Stop}
   RotationAngle=ScalarParameter() {defaultValue=0.3, parameterName=Edge4_Rotation}
   RotationCenter=AppendVector(
    A=ScalarParameter() {defaultValue=0.5, parameterName=Edge4_RotationCenter_X}
    B=ScalarParameter() {defaultValue=0.5, parameterName=Edge4_RotationCenter_Y}
   )
  ) {materialFunction=&MF_UI_LinearGradient}
  n8 = Saturate(None=DotProduct(A=OneMinus(None=@n6.Result), B=OneMinus(None=@n7.Result)))
  n9 = Saturate(None=DotProduct(A=Saturate(None=@n5), B=OneMinus(None=@n8)))
  n10 = MaterialFunctionCall(
   UVs=MaterialFunctionCall() {materialFunction=&ScreenAlignedUVs}."X 100%, Y 100%"
   Radius=ScalarParameter() {defaultValue=0.5, parameterName=VignetteRadius}
   Density=ScalarParameter() {defaultValue=2.333, parameterName=VignetteDensity}
  ) {materialFunction=&RadialGradientExponential}
  n11 = Saturate(None=@n10.RadialGradientExponential)
  n12 = StaticSwitchParameter(True=Saturate(None=DotProduct(A=@n9, B=@n11)), False=@n9) {parameterName=UseScreenSpaceFade}
  n13 = ScalarParameter() {defaultValue=1, parameterName=Final_Alpha}
  LinearInterpolate(
   A=VectorParameter() {parameterName=ColorA, defaultValue=<rgba 1 0 0 1>}.RGB
   B=VectorParameter() {parameterName=ColorB, defaultValue=<rgba 0 1 0 1>}.RGB
   Alpha=Saturate(None=Multiply(A=@n12, B=@n13))
  )
  MaterialFunctionCall() {materialFunction=/FortUILibrary/Foundation/Art_ToolKit/Tools/MF_UI_SquareToScreenSpace}
  OneMinus(None=@n10.RadialGradientExponential)
 ]
}
